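# MF_MojoBeatPulse is a material function that pulses an Emissive input to the music using Sparks audio parameter collection values.
# Inverted BeatLifetime blends between BeatPulse Min and Max to scale the emissive, fading from Max as that value rises.
# While MeasureLifetime is below a small threshold, the pulse is multiplied by BeatPulse Strength, accenting the measure start.
# BeatPulse_Active, a custom primitive data parameter, blends between unmodified and pulsed emissive per primitive.
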MaterialFunction /SparksMaps/Environment/MaterialFunctions/MF_MojoBeatPulse
graph {nodes [
  n1 = FunctionInput() {inputName=Emissive}
  n2 = CollectionParameter() {collection=/FMAudioReactiveData/AudioAnalysis/MPC_Sparks_Audio_Parameters, parameterName=MeasureLifetime}
  n3 = FunctionInput() {bUsePreviewValueAsDefault=true, inputName=BeatPulse Min, inputType=FunctionInput_Scalar, sortPriority=2, previewValue=<0 0 0 0>}
  n4 = FunctionInput() {bUsePreviewValueAsDefault=true, inputName=BeatPulse Max, inputType=FunctionInput_Scalar, sortPriority=3, previewValue=<1 1 1 1>}
  n5 = CollectionParameter() {collection=&MPC_Sparks_Audio_Parameters, parameterName=BeatLifetime}
  n6 = Multiply(A=@n1, B=LinearInterpolate(A=@n3, B=@n4, Alpha=OneMinus(None=@n5)))
  n7 = Multiply(
   A=FunctionInput(Preview=ScalarParameter() {defaultValue=5, group=MojoBeatPulse, parameterName=BeatPulse_Strength}) {bUsePreviewValueAsDefault=true, inputName=BeatPulse Strength, inputType=FunctionInput_Scalar, sortPriority=1}
   B=@n6
  )
  n8 = ScalarParameter() {bUseCustomPrimitiveData=true, defaultValue=1, group=MojoBeatPulse, parameterName=BeatPulse_Active, primitiveDataIndex=20}
  FunctionOutput(None=LinearInterpolate(A=@n1, B=If(A=@n2, "A > B"=@n6, "A < B"=@n7) {constB=0.1}, Alpha=@n8))
 ]}
